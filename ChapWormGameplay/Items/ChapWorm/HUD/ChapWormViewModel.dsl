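; ChapWorm HUD progress-bar view model that shows when the pawn has unlimited energy.
; Paired gameplay tag query waits on the pawn toggle the state, which force-shows and enables a full bar with an override material, or restores default materials.
; While active, a looping timer finds the longest remaining constant-regen gameplay effect on the cached ability system component and pulses the bar during its final seconds.
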(event EventInitialize
  (params (Pawn :type Pawn%20Object%20Reference))
  (|CallParentFunction Pawn)
  (bind _asfort_ability_system_component (Utilities|Casting|CastToFortAbilitySystemComponent (Ability|GetAbilitySystemComponent Pawn)))
  (Variables|Default|SetAbilitySystemComponent _asfort_ability_system_component)
  (Ability|Tasks|WaitGameplayTagQueryOnActor Pawn (Variables|Default|GetUnlimitedEnergyTagQuery)
    (:then
      (Ability|Tasks|WaitGameplayTagQueryOnActor Pawn (Variables|Default|GetUnlimitedEnergyTagQuery) "WhenFalse"
        (:Triggered
          (CallFunction|ToggleUnlimitedEnergyState false))))
    (:Triggered
      (CallFunction|ToggleUnlimitedEnergyState true))))

(event EventUninitialize
  (params (Pawn :type Pawn%20Object%20Reference))
  (Variables|Default|SetAbilitySystemComponent 0))

(event Custom|ToggleUnlimitedEnergyState
  (params (bUnlimitedEnergyState :type Boolean))
  (bind _output_get (Variables|GenericProgressBarViewModel|SetbForceVisible bUnlimitedEnergyState))
  (Variables|GenericProgressBarViewModel|SetbEnabled bUnlimitedEnergyState)
  (bind _output_get_2 (Variables|Default|SetbUnlimitedEnergyActive _output_get))
  (if _output_get_2
    (CallFunction|SetProgressBarOverrideMaterial "/TacticalSprintUI/UI/Materials/MI_UI_Rounded_HitPointBar_ChapWorm_OP.MI_UI_Rounded_HitPointBar_ChapWorm_OP")
    (Variables|GenericProgressBarViewModel|SetProgressValue (Variables|GenericProgressBarViewModel|GetProgressMaxValue))
    (bind _returnvalue (Utilities|Time|SetTimerbyEvent (delegate Custom|EnablePulsing) 0.25 true))
    (Variables|Default|SetUnlimitedEnergyTimerHandle _returnvalue)
    (else
      (CallFunction|RestoreDefaultMaterials)
      (Utilities|Time|ClearandInvalidateTimerbyHandle 0 (Variables|Default|GetUnlimitedEnergyTimerHandle))
      (Variables|GenericProgressBarViewModel|SetbIsPulsing false))))

(event Custom|EnablePulsing
  (bind _timeremaining (CallFunction|FindLongestTimeRemaining))
  (Variables|GenericProgressBarViewModel|SetbIsPulsing (and (<= _timeremaining 3.0) (> _timeremaining 0))))

(fn FindLongestTimeRemaining
  (bind _abilitysystemcomponent (Variables|Default|GetAbilitySystemComponent))
  (Utilities|IsValid _abilitysystemcomponent
    (:Is%20Valid
      (bind _returnvalue (GameplayEffects|GetActiveEffectswithAllTags _abilitysystemcomponent "(GameplayTags=((TagName=\"Asset.Athena.Energy.ConstantRegen\")))"))
      (for _array_element _returnvalue
        (bind _returnvalue_2 (Ability|GameplayEffect|GetActiveGameplayEffectRemainingDuration 0 _array_element))
        (if (> _returnvalue_2 (LocalVariables|GetLongestTimeFound))
          (LocalVariables|SetLongestTimeFound _returnvalue_2)))
      (return (LocalVariables|GetLongestTimeFound)))))
